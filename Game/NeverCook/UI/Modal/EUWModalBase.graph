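# Editor utility widget base for modal dialogs, with a title, scrollable content slot, close button and five configurable bottom buttons.
# SetupModal stores the owning menu anchor and context widget, inserts a supplied or newly created child widget, configures buttons by EAppMsgType and binds the result delegate.
# Clicks map to EAppReturnType; Close shuts the anchor, re-enables the context widget and broadcasts OnClosed with result and child.
# The MsgCategory switch has unconnected outputs.
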EditorUtilityWidgetBlueprint /Game/NeverCook/UI/Modal/EUWModalBase
graphs [
 Close {nodes [
   n1 = Close!() -> (then, DialogResult:EAppReturnType Enum) {is=title K2Node_FunctionEntry}
   > n2 = Set DialogResult(DialogResult:EAppReturnType Enum=@n1) -> (Output_Get:EAppReturnType Enum)
   n3 = Get Owner() -> (Owner:Menu Anchor Object Reference)
   > Is Valid(InputObject=@n3)
   > Close(self:Menu Anchor Object Reference=@n3) {is=title K2Node_CallFunction}
   > SetIsEnabled(self=Get ContextOwner() -> (ContextOwner:Widget Object Reference), bInIsEnabled=true)
   > Call On Closed!(
    execute:Exec
    self:EUWModal Base Object Reference
    ReturnType:EAppReturnType Enum=@n2
    InnerWidget:Editor Utility Widget Object Reference=Get ChildUI() -> (ChildUI:Editor Utility Widget Object Reference)
   ) -> (then:Exec) {is=title K2Node_CallDelegate}
  ]}
 EventGraph {nodes [
   "On Clicked (btnClose)"() {is=title K2Node_ComponentBoundEvent}
   > Close!(execute, self:Self Object Reference, DialogResult:EAppReturnType Enum=Cancel) -> (then)
   "On Clicked (btn0)"() {is=title K2Node_ComponentBoundEvent}
   n1 = Get MsgType() -> (MsgType:EAppMsgType Enum=Ok)
   > n2 = Switch on EAppMsgType!(execute, Selection:EAppMsgType Enum=@n1) -> (
    Ok
    YesNo:Exec
    OkCancel:Exec
    YesNoCancel:Exec
    CancelRetryContinue:Exec
    YesNoYesAllNoAll:Exec
    YesNoYesAllNoAllCancel:Exec
    YesNoYesAll:Exec
   ) {is=title K2Node_SwitchEnum}
   Close(execute=@n2.CancelRetryContinue, DialogResult=Cancel)
   Close(execute=[@n2.Ok, @n2.OkCancel], DialogResult=Ok)
   Close(execute=[@n2.YesNo, @n2.YesNoCancel, @n2.YesNoYesAll, @n2.YesNoYesAllNoAll, @n2.YesNoYesAllNoAllCancel], DialogResult=Yes)
   "On Clicked (btn1)"() {is=title K2Node_ComponentBoundEvent}
   > n3 = Switch on EAppMsgType(Selection=@n1)
   Close(execute=@n3.CancelRetryContinue, DialogResult=Retry)
   Close(execute=[@n3.OkCancel, @n4.YesNoCancel], DialogResult=Cancel)
   Close(execute=[@n3.YesNo, @n3.YesNoCancel, @n3.YesNoYesAll, @n3.YesNoYesAllNoAll, @n3.YesNoYesAllNoAllCancel], DialogResult=No)
   "On Clicked (btn2)"() {is=title K2Node_ComponentBoundEvent}
   > n4 = Switch on EAppMsgType(Selection=@n1)
   Close(execute=@n4.CancelRetryContinue, DialogResult=Continue)
   Close(execute=[@n4.YesNoYesAll, @n4.YesNoYesAllNoAll, @n4.YesNoYesAllNoAllCancel], DialogResult=YesAll)
   "On Clicked (btn3)"() {is=title K2Node_ComponentBoundEvent}
   > n5 = Switch on EAppMsgType(Selection=@n1)
   Close(execute=[@n5.YesNoYesAllNoAll, @n5.YesNoYesAllNoAllCancel], DialogResult=NoAll)
   "On Clicked (btn4)"() {is=title K2Node_ComponentBoundEvent}
   > n6 = Switch on EAppMsgType(Selection=@n1)
   Close(execute=@n6.YesNoYesAllNoAllCancel, DialogResult=Cancel)
   n7 = OnSet!() -> (OutputDelegate, then, MsgType:EAppMsgType Enum=Ok, MsgCategory:EAppMsgCategory Enum=Warning) {is=title K2Node_CustomEvent}
   > n8 = Set MsgType(MsgType:EAppMsgType Enum=@n7.MsgType) -> (Output_Get:EAppMsgType Enum=Ok)
   > n9 = Set MsgCategory(MsgCategory:EAppMsgCategory Enum=@n7.MsgCategory) -> (Output_Get:EAppMsgCategory Enum=Warning)
   > n10 = Sequence()
   > n11 = Switch on EAppMsgType(Selection=@n8)
   SetButtons!(
    execute=@n11.CancelRetryContinue
    self
    btn0_text:Text="LOCTEXT(\"Cancel\")"
    btn0_visibility:ESlateVisibility Enum=Visible
    btn1_text:Text="LOCTEXT(\"Retry\")"
    btn1_visibility:ESlateVisibility Enum=Visible
    btn2_text:Text="LOCTEXT(\"Continue\")"
    btn2_visibility:ESlateVisibility Enum=Visible
    btn3_text:Text
    btn3_visibility:ESlateVisibility Enum=Collapsed
    btn4_text:Text
    btn4_visibility:ESlateVisibility Enum=Collapsed
   ) -> (then)
   SetButtons(
    execute=@n11.Ok
    btn0_text="LOCTEXT(\"OK\")"
    btn0_visibility=Visible
    btn1_visibility=Collapsed
    btn2_visibility=Collapsed
    btn3_visibility=Collapsed
    btn4_visibility=Collapsed
   )
   SetButtons(
    execute=@n11.OkCancel
    btn0_text="LOCTEXT(\"Ok\")"
    btn0_visibility=Visible
    btn1_text="LOCTEXT(\"Cancel\")"
    btn1_visibility=Visible
    btn2_visibility=Collapsed
    btn3_visibility=Collapsed
    btn4_visibility=Collapsed
   )
   SetButtons(
    execute=@n11.YesNo
    btn0_text="LOCTEXT(\"Yes\")"
    btn0_visibility=Visible
    btn1_text="LOCTEXT(\"No\")"
    btn1_visibility=Visible
    btn2_visibility=Collapsed
    btn3_visibility=Collapsed
    btn4_visibility=Collapsed
   )
   SetButtons(
    execute=@n11.YesNoCancel
    btn0_text="LOCTEXT(\"Yes\")"
    btn0_visibility=Visible
    btn1_text="LOCTEXT(\"No\")"
    btn1_visibility=Visible
    btn2_text="LOCTEXT(\"Cancel\")"
    btn2_visibility=Visible
    btn3_visibility=Collapsed
    btn4_visibility=Collapsed
   )
   SetButtons(
    execute=@n11.YesNoYesAll
    btn0_text="LOCTEXT(\"Yes\")"
    btn0_visibility=Visible
    btn1_text="LOCTEXT(\"No\")"
    btn1_visibility=Visible
    btn2_text="LOCTEXT(\"Yes All\")"
    btn2_visibility=Visible
    btn3_visibility=Collapsed
    btn4_visibility=Collapsed
   )
   SetButtons(
    execute=@n11.YesNoYesAllNoAll
    btn0_text="LOCTEXT(\"Yes\")"
    btn0_visibility=Visible
    btn1_text="LOCTEXT(\"No\")"
    btn1_visibility=Visible
    btn2_text="LOCTEXT(\"Yes All\")"
    btn2_visibility=Visible
    btn3_text="LOCTEXT(\"No All\")"
    btn3_visibility=Visible
    btn4_visibility=Collapsed
   )
   SetButtons(
    execute=@n11.YesNoYesAllNoAllCancel
    btn0_text="LOCTEXT(\"Yes\")"
    btn0_visibility=Visible
    btn1_text="LOCTEXT(\"No\")"
    btn1_visibility=Visible
    btn2_text="LOCTEXT(\"Yes All\")"
    btn2_visibility=Visible
    btn3_text="LOCTEXT(\"No All\")"
    btn3_visibility=Visible
    btn4_text="LOCTEXT(\"Cancel\")"
    btn4_visibility=Visible
   )
   Switch on EAppMsgCategory!(execute=@n10.then_1, Selection:EAppMsgCategory Enum=@n9) -> (Warning:Exec, Error:Exec, Success:Exec, Info:Exec) {is=title K2Node_SwitchEnum}
   Event Construct()
   > Clear(TargetArray:Array of Editor Utility Button Object References=Get Buttons() -> (Buttons:Array of Editor Utility Button Object References))
   > Set Buttons(Buttons:Array of Editor Utility Button Object References=Make Array!(
     "[0]":Editor Utility Button Object Reference=Get btn0() -> (btn0:Editor Utility Button Object Reference)
     "[1]":Editor Utility Button Object Reference=Get btn1() -> (btn1:Editor Utility Button Object Reference)
     "[2]":Editor Utility Button Object Reference=Get btn2() -> (btn2:Editor Utility Button Object Reference)
     "[3]":Editor Utility Button Object Reference=Get btn3() -> (btn3:Editor Utility Button Object Reference)
     "[4]":Editor Utility Button Object Reference=Get btn4() -> (btn4:Editor Utility Button Object Reference)
    ) -> (Array:Array of Editor Utility Button Object References)) -> (Output_Get:Array of Editor Utility Button Object References)
   > Set Texts(Texts:Array of Widget Object References=Make Array(
     "[0]":Widget Object Reference=Get txt0() -> (txt0:Text Object Reference)
     "[1]":Widget Object Reference=Get txt1() -> (txt1:Text Object Reference)
     "[2]":Widget Object Reference=Get txt2() -> (txt2:Text Object Reference)
     "[3]":Widget Object Reference=Get txt3() -> (txt3:Text Object Reference)
     "[4]":Widget Object Reference=Get txt4() -> (txt4:Text Object Reference)
    ) -> (Array:Array of Widget Object References)) -> (Output_Get:Array of Widget Object References)
   > SetButtons(
    btn0_text="LOCTEXT(\"Ok\")"
    btn0_visibility=Visible
    btn1_visibility=Collapsed
    btn2_visibility=Collapsed
    btn3_visibility=Collapsed
    btn4_visibility=Collapsed
   )
   Event PreConstruct()
   Event Tick()
  ]}
 OnClosed {nodes [OnClosed!() -> (then, ReturnType:EAppReturnType Enum, InnerWidget:Editor Utility Widget Object Reference) {is=title K2Node_FunctionEntry}]}
 SetButtons {nodes [
   n1 = SetButtons!() -> (
    then
    btn0_text:Text="LOCTEXT(\"Ok\")"
    btn0_visibility:ESlateVisibility Enum=Visible
    btn1_text:Text
    btn1_visibility:ESlateVisibility Enum=Collapsed
    btn2_text:Text
    btn2_visibility:ESlateVisibility Enum=Collapsed
    btn3_text:Text
    btn3_visibility:ESlateVisibility Enum=Collapsed
    btn4_text:Text
    btn4_visibility:ESlateVisibility Enum=Collapsed
   ) {is=title K2Node_FunctionEntry}
   > n2 = For Each Loop(Array:Array of Editor Utility Button Object References=Get Buttons()) -> (Array Element:Editor Utility Button Object Reference)
   > Set Visibility(
    Visibility="Get (a ref)"(
     Array:Array of ESlateVisibility Enums=Make Array(
      "[0]":ESlateVisibility Enum=@n1.btn0_visibility
      "[1]":ESlateVisibility Enum=@n1.btn1_visibility
      "[2]":ESlateVisibility Enum=@n1.btn2_visibility
      "[3]":ESlateVisibility Enum=@n1.btn3_visibility
      "[4]":ESlateVisibility Enum=@n1.btn4_visibility
     ) -> (Array:Array of ESlateVisibility Enums)
     Dimension 1=@n2.Array Index
    ) -> (Output:"ESlateVisibility Enum (by ref)")
    self=@n2
   )
   > n3 = Cast To TextBlock!(
    execute
    Object="Get (a copy)"(Array:Array of Widget Object References=Get Texts() -> (Texts:Array of Widget Object References), Dimension 1=@n2.Array Index) -> (Output:Widget Object Reference)
   ) -> (then, CastFailed, AsText:Text Object Reference) {is=title K2Node_DynamicCast}
   > Set Text(
    Text="Get (a copy)"(
     Array:Array of Texts=Make Array("[0]":Text=@n1, "[1]":Text=@n1.btn1_text, "[2]":Text=@n1.btn2_text, "[3]":Text=@n1.btn3_text, "[4]":Text=@n1.btn4_text) -> (Array:Array of Texts)
     Dimension 1=@n2.Array Index
    ) -> (Output:Text)
    self=@n3
   )
  ]}
 SetUI {nodes [
   n1 = SetUI!() -> (then, Widget Class:Editor Utility Widget Class Reference, Widget Instance:Editor Utility Widget Object Reference) {is=title K2Node_FunctionEntry}
   n2 = Get UI() -> (UI:Named Slot Object Reference)
   > ClearChildren(self=@n2)
   > n3 = Is Valid(InputObject=@n1.Widget Instance)
   n4 = Create Widget!(
    execute:Exec=@n3.Is Not Valid
    WorldContextObject:Object Reference
    Class:User Widget Class Reference=@n1
    OwningPlayer:Player Controller Object Reference
   ) -> (then:Exec, ReturnValue:Editor Utility Widget Object Reference)
   > AddChild(self=@n2, Content=@n4)
   > n5 = Set ChildUI(ChildUI:Editor Utility Widget Object Reference=@n4) -> (Output_Get:Editor Utility Widget Object Reference)
   > Return Node!(execute, Widget:Editor Utility Widget Object Reference=@n5)
   AddChild(execute=@n3.Is Valid, self=@n2, Content=@n1.Widget Instance)
   > n6 = Set ChildUI(ChildUI=@n1.Widget Instance)
   > Return Node(Widget=@n6)
  ]}
 SetupModal {nodes [
   n1 = SetupModal!() -> (
    then
    Title:Text
    AnchorOwner:Menu Anchor Object Reference
    ContextOwner:Widget Object Reference
    Widget Class:Editor Utility Widget Class Reference
    Widget Instance:Editor Utility Widget Object Reference
    MsgType:EAppMsgType Enum=Ok
    MsgCategory:EAppMsgCategory Enum=Warning
    Event:"Delegate (by ref)"
   ) {is=title K2Node_FunctionEntry}
   > n2 = Sequence!(execute) -> (then_0, then_1, then_2:Exec, then_3:Exec, then_4:Exec, then_5:Exec, then_6:Exec, then_7:Exec)
   > Set Text(Text=@n1, self=Get Title() -> (Title:Text Object Reference))
   Set Owner(execute=@n2.then_1, Owner:Menu Anchor Object Reference=@n1.AnchorOwner) -> (Output_Get:Menu Anchor Object Reference)
   Set ContextOwner(execute=@n2.then_2, ContextOwner:Widget Object Reference=@n1.ContextOwner) -> (Output_Get:Widget Object Reference)
   n3 = SetUI!(
    execute=@n2.then_3
    self
    Widget Class:Editor Utility Widget Class Reference=@n1.Widget Class
    Widget Instance:Editor Utility Widget Object Reference=@n1.Widget Instance
   ) -> (then, Widget:Editor Utility Widget Object Reference) {is=title K2Node_CallFunction}
   OnSet!(execute=@n2.then_4, self, MsgType:EAppMsgType Enum=@n1.MsgType, MsgCategory:EAppMsgCategory Enum=@n1.MsgCategory) -> (then) {is=title K2Node_CallFunction}
   Bind Event to On Closed(execute=@n2.then_5, self:EUWModal Base Object Reference, Delegate=@n1.Event) {is=title K2Node_AddDelegate}
   Return Node(execute=@n2.then_6, Widget=@n3)
  ]}
]
variables [
 Buttons {type="array<EditorUtilityButton>", default []}
 ChildUI {type=EditorUtilityWidget, default=null}
 ContextOwner {type=Widget, default=null}
 DialogResult {type=EAppReturnType, default=No}
 MsgCategory {type=EAppMsgCategory, default=Info}
 MsgType {type=EAppMsgType, default=Ok}
 OnClosed {default=()}
 Owner {type=MenuAnchor, default=null}
 Texts {type="array<Widget>", default []}
]
widgets {
 info {rootWidgetClass=&Border, widgetCount=20}
 widgets [
  Border_40 {
   widgetClassPath=&Border
   widget=Border_40
   properties {
    background {
     bIsDynamicallyLoaded=false
     drawAs=RoundedBox
     imageType=NoImage
     mirroring=NoMirror
     resourceName=None
     resourceObject=null
     tiling=NoTile
     imageSize=<32 32>
     margin=<ltrb 0 0 1 1>
     outlineSettings {
      bUseBrushTransparency=false
      roundingType=FixedRadius
      width=1
      color {colorUseRule=UseColor_Specified, specifiedColor=<rgba 1 1 1 1>}
      cornerRadii=<4 4 4 4>
     }
     tintColor {colorUseRule=UseColor_Foreground, specifiedColor=<rgba 1 1 1 1>}
     uVRegion {bIsValid=false, max=<0 0>, min=<0 0>}
    }
    brushColor=<rgba 0.0395 0.0395 0.0395 1>
    padding=<ltrb 4 4 4 4>
   }
  }
  VerticalBox_33 {
   widgetClassPath=&VerticalBox
   parent=Border_40
   widget=VerticalBox_33
   slot {class=BorderSlot, properties {padding=<ltrb 4 4 4 4>}}
  }
  HorizontalBox {
   widgetClassPath=&HorizontalBox
   parent=VerticalBox_33
   widget=HorizontalBox
   slot {class=VerticalBoxSlot, properties {verticalAlignment=VAlign_Top, padding=<ltrb 0 0 0 8>}}
  }
  Title {
   widgetClassPath=&TextBlock
   bIsVariable=true
   parent=HorizontalBox
   widget=Title
   properties {
    text=Default Title
    font {
     bForceMonospaced=false
     bMaterialIsStencil=false
     fontMaterial=null
     fontObject=&Roboto
     letterSpacing=0
     monospacedWidth=1
     size=13.5
     skewAmount=0
     typefaceFontName=Light
     outlineSettings {
      bApplyOutlineToDropShadows=false
      bMiteredCorners=false
      bSeparateFillAlpha=false
      outlineMaterial=null
      outlineSize=0
      outlineColor=<rgba 0 0 0 1>
     }
    }
   }
   slot {
    class=HorizontalBoxSlot
    properties {horizontalAlignment=HAlign_Left, verticalAlignment=VAlign_Top, padding=<ltrb 4 0 0 0>, size {sizeRule=Fill, value=1}}
   }
  }
  btnClose {
   widgetClassPath=&EditorUtilityButton
   parent=HorizontalBox
   widget=btnClose
   properties {
    backgroundColor=<rgba 1 1 1 0>
    widgetStyle {
     clickedSlateSound {resourceObject=null}
     disabled {
      bIsDynamicallyLoaded=false
      drawAs=RoundedBox
      imageType=NoImage
      mirroring=NoMirror
      resourceName=None
      resourceObject=null
      tiling=NoTile
      imageSize=<0 0>
      margin=<ltrb 0 0 0 0>
      outlineSettings {
       bUseBrushTransparency=false
       roundingType=FixedRadius
       width=1
       color {colorUseRule=UseColor_ColorTable, specifiedColor=<rgba 0 0 0 0>}
       cornerRadii=<4 4 4 4>
      }
      tintColor {colorUseRule=UseColor_Specified, specifiedColor=<rgba 0.0395 0.0395 0.0395 1>}
      uVRegion {bIsValid=false, max=<0 0>, min=<0 0>}
     }
     disabledForeground {colorUseRule=UseColor_ColorTable, specifiedColor=<rgba 0 0 0 0>}
     hovered {
      bIsDynamicallyLoaded=false
      drawAs=RoundedBox
      imageType=NoImage
      mirroring=NoMirror
      resourceName=None
      resourceObject=null
      tiling=NoTile
      imageSize=<32 32>
      margin=<ltrb 0 0 0 0>
      outlineSettings {
       bUseBrushTransparency=false
       roundingType=FixedRadius
       width=1
       color {colorUseRule=UseColor_ColorTable, specifiedColor=<rgba 0 0 0 0>}
       cornerRadii=<4 4 4 4>
      }
      tintColor {colorUseRule=UseColor_Specified, specifiedColor=<rgba 0.0953 0.0953 0.0953 1>}
      uVRegion {bIsValid=false, max=<0 0>, min=<0 0>}
     }
     hoveredForeground {colorUseRule=UseColor_ColorTable, specifiedColor=<rgba 0 0 0 0>}
     hoveredSlateSound {resourceObject=null}
     normal {
      bIsDynamicallyLoaded=false
      drawAs=NoDrawType
      imageType=NoImage
      mirroring=NoMirror
      resourceName=None
      resourceObject=null
      tiling=NoTile
      imageSize=<32 32>
      margin=<ltrb 0 0 0 0>
      outlineSettings {
       bUseBrushTransparency=false
       roundingType=FixedRadius
       width=1
       color {colorUseRule=UseColor_ColorTable, specifiedColor=<rgba 0 0 0 0>}
       cornerRadii=<4 4 4 4>
      }
      tintColor {colorUseRule=UseColor_Specified, specifiedColor=<rgba 0.0395 0.0395 0.0395 1>}
      uVRegion {bIsValid=false, max=<0 0>, min=<0 0>}
     }
     normalForeground {colorUseRule=UseColor_ColorTable, specifiedColor=<rgba 0 0 0 0>}
     normalPadding=<ltrb 4 0 4 0>
     pressed {
      bIsDynamicallyLoaded=false
      drawAs=RoundedBox
      imageType=NoImage
      mirroring=NoMirror
      resourceName=None
      resourceObject=null
      tiling=NoTile
      imageSize=<32 32>
      margin=<ltrb 0 0 0 0>
      outlineSettings {
       bUseBrushTransparency=false
       roundingType=FixedRadius
       width=1
       color {colorUseRule=UseColor_ColorTable, specifiedColor=<rgba 0 0 0 0>}
       cornerRadii=<4 4 4 4>
      }
      tintColor {colorUseRule=UseColor_Specified, specifiedColor=<rgba 0.0284 0.0284 0.0284 1>}
      uVRegion {bIsValid=false, max=<0 0>, min=<0 0>}
     }
     pressedForeground {colorUseRule=UseColor_ColorTable, specifiedColor=<rgba 0 0 0 0>}
     pressedPadding=<ltrb 4 1.5 4 0.5>
     pressedSlateSound {resourceObject=null}
    }
   }
   slot {class=HorizontalBoxSlot, properties {horizontalAlignment=HAlign_Right}}
  }
  closeIcon {
   widgetClassPath=&Image
   parent=btnClose
   widget=closeIcon
   properties {brush {
     bIsDynamicallyLoaded=false
     drawAs=Image
     imageType=FullColor
     mirroring=NoMirror
     resourceName=None
     resourceObject=/FortniteCoreUI/Frontend/StoreMain/Materials/Textures/T_UI_Icon_XClose
     tiling=NoTile
     imageSize=<20 20>
     margin=<ltrb 0 0 0 0>
     outlineSettings {
      bUseBrushTransparency=false
      roundingType=HalfHeightRadius
      width=0
      color {colorUseRule=UseColor_Specified, specifiedColor=<rgba 0 0 0 0>}
      cornerRadii=<0 0 0 0>
     }
     tintColor {colorUseRule=UseColor_Specified, specifiedColor=<rgba 1 0.1507 0.145 1>}
     uVRegion {bIsValid=false, max=<0 0>, min=<0 0>}
    }}
   slot {class=ButtonSlot}
  }
  CanvasPanel_0 {
   widgetClassPath=&CanvasPanel
   parent=VerticalBox_33
   widget=CanvasPanel_0
   slot {class=VerticalBoxSlot, properties {padding=<ltrb 24 24 24 24>, size {sizeRule=Fill, value=1}}}
  }
  EditorUtilityScrollBox_0 {
   widgetClassPath=&EditorUtilityScrollBox
   parent=CanvasPanel_0
   widget=EditorUtilityScrollBox_0
   slot {
    class=CanvasPanelSlot
    properties {bAutoSize=true, layoutData {alignment=<0 0>, anchors {maximum=<1 1>, minimum=<0 0>}, offsets=<ltrb 0 0 0 0>}}
   }
  }
  UI {widgetClassPath=&NamedSlot, bIsVariable=true, parent=EditorUtilityScrollBox_0, widget=UI, slot {class=ScrollBoxSlot}}
  HorizontalBox_57 {
   widgetClassPath=&HorizontalBox
   parent=VerticalBox_33
   widget=HorizontalBox_57
   slot {class=VerticalBoxSlot, properties {horizontalAlignment=HAlign_Right, verticalAlignment=VAlign_Bottom, padding=<ltrb 0 8 8 8>}}
  }
  btn0 {
   widgetClassPath=&EditorUtilityButton
   parent=HorizontalBox_57
   widget=btn0
   slot {class=HorizontalBoxSlot, properties {horizontalAlignment=HAlign_Right, padding=<ltrb 0 0 8 0>}}
  }
  txt0 {
   widgetClassPath=&TextBlock
   bIsVariable=true
   parent=btn0
   widget=txt0
   properties {
    justification=Center
    font {
     bForceMonospaced=false
     bMaterialIsStencil=false
     fontMaterial=null
     fontObject=&Roboto
     letterSpacing=0
     monospacedWidth=1
     size=15
     skewAmount=0
     typefaceFontName=Regular
     outlineSettings {
      bApplyOutlineToDropShadows=false
      bMiteredCorners=false
      bSeparateFillAlpha=false
      outlineMaterial=null
      outlineSize=0
      outlineColor=<rgba 0 0 0 1>
     }
    }
   }
   slot {class=ButtonSlot, properties {horizontalAlignment=HAlign_Left, verticalAlignment=VAlign_Top, padding=<ltrb 10 4 10 4>}}
  }
  btn1 {
   widgetClassPath=&EditorUtilityButton
   parent=HorizontalBox_57
   widget=btn1
   slot {class=HorizontalBoxSlot, properties {padding=<ltrb 0 0 8 0>}}
  }
  txt1 {
   widgetClassPath=&TextBlock
   bIsVariable=true
   parent=btn1
   widget=txt1
   properties {
    justification=Center
    font {
     bForceMonospaced=false
     bMaterialIsStencil=false
     fontMaterial=null
     fontObject=&Roboto
     letterSpacing=0
     monospacedWidth=1
     size=15
     skewAmount=0
     typefaceFontName=Regular
     outlineSettings {
      bApplyOutlineToDropShadows=false
      bMiteredCorners=false
      bSeparateFillAlpha=false
      outlineMaterial=null
      outlineSize=0
      outlineColor=<rgba 0 0 0 1>
     }
    }
   }
   slot {class=ButtonSlot, properties {horizontalAlignment=HAlign_Left, verticalAlignment=VAlign_Top, padding=<ltrb 10 4 10 4>}}
  }
  btn2 {
   widgetClassPath=&EditorUtilityButton
   parent=HorizontalBox_57
   widget=btn2
   slot {class=HorizontalBoxSlot, properties {padding=<ltrb 0 0 8 0>}}
  }
  txt2 {
   widgetClassPath=&TextBlock
   bIsVariable=true
   parent=btn2
   widget=txt2
   properties {
    justification=Center
    font {
     bForceMonospaced=false
     bMaterialIsStencil=false
     fontMaterial=null
     fontObject=&Roboto
     letterSpacing=0
     monospacedWidth=1
     size=15
     skewAmount=0
     typefaceFontName=Regular
     outlineSettings {
      bApplyOutlineToDropShadows=false
      bMiteredCorners=false
      bSeparateFillAlpha=false
      outlineMaterial=null
      outlineSize=0
      outlineColor=<rgba 0 0 0 1>
     }
    }
   }
   slot {class=ButtonSlot, properties {horizontalAlignment=HAlign_Left, verticalAlignment=VAlign_Top, padding=<ltrb 10 4 10 4>}}
  }
  btn3 {
   widgetClassPath=&EditorUtilityButton
   parent=HorizontalBox_57
   widget=btn3
   slot {class=HorizontalBoxSlot, properties {padding=<ltrb 0 0 8 0>}}
  }
  txt3 {
   widgetClassPath=&TextBlock
   bIsVariable=true
   parent=btn3
   widget=txt3
   properties {
    justification=Center
    font {
     bForceMonospaced=false
     bMaterialIsStencil=false
     fontMaterial=null
     fontObject=&Roboto
     letterSpacing=0
     monospacedWidth=1
     size=15
     skewAmount=0
     typefaceFontName=Regular
     outlineSettings {
      bApplyOutlineToDropShadows=false
      bMiteredCorners=false
      bSeparateFillAlpha=false
      outlineMaterial=null
      outlineSize=0
      outlineColor=<rgba 0 0 0 1>
     }
    }
   }
   slot {class=ButtonSlot, properties {horizontalAlignment=HAlign_Left, verticalAlignment=VAlign_Top, padding=<ltrb 10 4 10 4>}}
  }
  btn4 {widgetClassPath=&EditorUtilityButton, parent=HorizontalBox_57, widget=btn4, slot {class=HorizontalBoxSlot}}
  txt4 {
   widgetClassPath=&TextBlock
   bIsVariable=true
   parent=btn4
   widget=txt4
   properties {
    justification=Center
    font {
     bForceMonospaced=false
     bMaterialIsStencil=false
     fontMaterial=null
     fontObject=&Roboto
     letterSpacing=0
     monospacedWidth=1
     size=15
     skewAmount=0
     typefaceFontName=Regular
     outlineSettings {
      bApplyOutlineToDropShadows=false
      bMiteredCorners=false
      bSeparateFillAlpha=false
      outlineMaterial=null
      outlineSize=0
      outlineColor=<rgba 0 0 0 1>
     }
    }
   }
   slot {class=ButtonSlot, properties {horizontalAlignment=HAlign_Left, verticalAlignment=VAlign_Top, padding=<ltrb 10 4 10 4>}}
  }
 ]
}
