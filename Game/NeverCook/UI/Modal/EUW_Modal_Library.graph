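# Editor utility library with ShowEUWModal, which shows a modal dialog inside editor utility widget UI.
# It opens the given Menu Anchor, then searches all EUWModalBase widgets and keeps the last visible one, since the anchor's spawned widget is not returned directly.
# SetupModal receives the title, button and category types, content class or instance, and close delegate.
# The function then disables the context widget and returns the created content widget.
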EditorUtilityBlueprint /Game/NeverCook/UI/Modal/EUW_Modal_Library
graphs [ShowEUWModal {nodes [
   n1 = ShowEUWModal!() -> (
    then
    Context:Widget Object Reference
    Title:Text="LOCTEXT(\"DialogTitle\")"
    Anchor:Menu Anchor Object Reference
    DialogUI:Editor Utility Widget Class Reference
    Buttons:EAppMsgType Enum=Ok
    Category:EAppMsgCategory Enum=Warning
    OnCloseEvent:"Delegate (by ref)"
    PassInWidget:Editor Utility Widget Object Reference
   ) {is=title K2Node_FunctionEntry}
   > Open!(execute, self:Menu Anchor Object Reference=@n1.Anchor, bFocusMenu:bool=true) -> (then)
   > n2 = GetAllWidgetsOfClass!(
    execute
    WorldContextObject:Object Reference
    WidgetClass:User Widget Class Reference=&/EUWModalBase.EUWModalBase_C
    TopLevelOnly:bool=false
   ) -> (then, FoundWidgets:Array of EUWModal Base Object References)
   > Branch(Condition=Is Not Empty!(TargetArray:Array of EUWModal Base Object References=@n2) -> (ReturnValue:bool))
   > n3 = Sequence()
   > n4 = For Each Loop(Array:Array of EUWModal Base Object References=@n2) -> (Array Element:EUWModal Base Object Reference)
   > Branch(Condition=IsVisible(self=@n4))
   > Set ModalBase(ModalBase:EUWModal Base Object Reference=@n4) -> (Output_Get:EUWModal Base Object Reference)
   n5 = Get ModalBase() -> (ModalBase:EUWModal Base Object Reference)
   Is Valid(exec=@n3.then_1, InputObject=@n5)
   > n6 = SetupModal!(
    execute
    self:EUWModal Base Object Reference=@n5
    Title:Text=Get Title() -> (Title:Text)
    AnchorOwner:Menu Anchor Object Reference=Get Anchor() -> (Anchor:Menu Anchor Object Reference)
    ContextOwner:Widget Object Reference=Get Context() -> (Context:Widget Object Reference)
    Widget Class:Editor Utility Widget Class Reference=Get DialogUI() -> (DialogUI:Editor Utility Widget Class Reference)
    Widget Instance:Editor Utility Widget Object Reference=@n1.PassInWidget
    MsgType:EAppMsgType Enum=Get Buttons() -> (Buttons:EAppMsgType Enum=Ok)
    MsgCategory:EAppMsgCategory Enum=Get Category() -> (Category:EAppMsgCategory Enum=Warning)
    Event:"Delegate (by ref)"=@n1.OnCloseEvent
   ) -> (then, Widget:Editor Utility Widget Object Reference)
   > SetIsEnabled(self=Get Context(), bInIsEnabled=false)
   > Return Node!(execute, CreatedWidget:Editor Utility Widget Object Reference=@n6)
  ]}]
variables []
